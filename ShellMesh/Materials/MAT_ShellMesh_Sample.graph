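# Sample material for the ShellMesh plugin, demonstrating its custom shell layer and shell offset expressions on a Default Lit surface.
# The normalized shell layer passes through a vertex interpolator to blend from red to green across the shell stack.
# The local shell offset is transformed out of local space in a separate branch, apparently to displace each shell's vertices.
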Material /ShellMesh/Materials/MAT_ShellMesh_Sample
graph {
 properties {usedShadingModels=Default Lit}
 nodes [
  LinearInterpolate(
   A=Constant3Vector() {constant=<rgba 1 0 0 0>}
   B=Constant3Vector() {constant=<rgba 0 1 0 0>}
   Alpha=VertexInterpolator(VS=ShellMeshMaterialExpressionNormalizedShellLayer() {is=class ShellMeshMaterialExpressionNormalizedShellLayer}).PS
  )
  Transform(None=ShellMeshMaterialExpressionLocalShellOffset() {is=class ShellMeshMaterialExpressionLocalShellOffset}) {transformSourceType=TRANSFORMSOURCE_Local}
 ]
}
