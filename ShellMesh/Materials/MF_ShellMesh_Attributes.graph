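# Material function that distributes and culls shell mesh layers, outputting a remapped shell index, a linear shell index and a world-space offset vector.
# Cull Factor blends how strongly ScaleFactor reduces the active layer count, renormalizing layer indices before a power curve redistributes them.
# The offset scales the local shell vector by ScaleFactor and remapped index, becoming NaN for culled layers.
# Vertex interpolators pass all outputs to the pixel shader.
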MaterialFunction /ShellMesh/Materials/MF_ShellMesh_Attributes
graph {
 properties {bExposeToLibrary=true, description=Helper function for shell layer distribution and culling}
 nodes [
  n1 = FunctionInput() {
   bUsePreviewValueAsDefault=true
   description="Controls how points will be distributed from the base to the edge of the volume (1 => linear, < 1 => more samples at edge, > 1 => more samples at base)"
   inputName=ShellDistributionCurvePower
   inputType=FunctionInput_Scalar
   previewValue=<1 0 0 1>
  }
  n2 = ShellMeshMaterialExpressionShellCount() {is=class ShellMeshMaterialExpressionShellCount}
  n3 = FunctionInput() {
   bUsePreviewValueAsDefault=true
   description=Scale factor for the maximum length of the offset vector
   inputName=ScaleFactor
   inputType=FunctionInput_Scalar
   previewValue=<1 0 0 1>
  }
  n4 = FunctionInput() {
   bUsePreviewValueAsDefault=true
   description="0 -> no culling; 1 -> cull based on the ScaleFactor"
   inputName=Cull Factor
   inputType=FunctionInput_Scalar
  }
  n5 = LinearInterpolate(B=@n3, Alpha=@n4) {constA=1}
  n6 = Min(A=@n2, B=Ceil(None=Multiply(A=@n2, B=@n5)))
  n7 = Multiply(A=Divide(A=ShellMeshMaterialExpressionShellLayerIndex() {is=class ShellMeshMaterialExpressionShellLayerIndex}, B=Subtract(A=@n2)), B=@n2)
  n8 = Custom!(None) {code="const float fNaN = 0.0f / 0.0f;\r\nreturn fNaN;", showCode=true}
  n9 = If(A=@n6, "A > B"=Divide(A=Floor(None=@n7), B=@n6), "A == B"=@n8, "A < B"=@n8)
  n10 = Power(Base=@n9, Exp=@n1)
  FunctionOutput(None=ShaderStageSwitch(PixelShader=VertexInterpolator(VS=@n10).PS, VertexShader=@n10)) {
   description="The remapped normalized shell index (0-1) based on the supplied distribution curve.  May be NaN if the shell has been culled."
   outputName=RemappedNormalizedShell
  }
  n11 = If(A=@n10, "A > B"=@n8, "A == B"=@n10, "A < B"=@n10) {constB=1}
  n12 = Transform(None=Multiply(A=ShellMeshMaterialExpressionLocalShellVector() {is=class ShellMeshMaterialExpressionLocalShellVector}, B=Multiply(A=@n3, B=@n11)))
  FunctionOutput(None=ShaderStageSwitch(PixelShader=VertexInterpolator(VS=@n12).PS, VertexShader=@n12)) {
   description="World offset vector for the current shell taking into consideration the supplied scaling && distribution.  May be NaN if culled."
   outputName=WorldShellOffset
  }
  FunctionOutput(None=ShaderStageSwitch(PixelShader=VertexInterpolator(VS=@n9).PS, VertexShader=@n9)) {
   description="Linear normalized shell index (0-1) for the current shell.  May be NaN if the shell is culled."
   outputName=LinearNormalizedShell
  }
 ]
}
